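# Physics asset for the badge on the F_MED_Barefoot female player body, defining a dangling bone chain for physics-driven secondary motion.
# Four thin box bodies (dyn_badge1 to dyn_badge4) link in series from a small master capsule at spine_04, with tight swing and tighter twist limits between links.
# Capsules on spine_03 and spine_04 approximate the torso, while the master-to-spine and spine-to-spine joints leave rotation free.
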PhysicsAsset /BRCosmetics/Characters/Player/Female/Medium/Bodies/F_MED_Barefoot/Meshes/F_MED_Barefoot_Badge_Physics
bodies {
 dyn_badge1 [{shapeType=Box, extentX=5.157, extentY=0.515, extentZ=9.8795, center=<2.435 0.0015 0>}]
 dyn_badge2 [{shapeType=Box, extentX=4.2481, extentY=0.7859, extentZ=8.8762, center=<2.1374 0.0469 0>}]
 dyn_badge3 [{shapeType=Box, extentX=4.9068, extentY=0.876, extentZ=6.6875, center=<2.3509 0.0051 -0.0032>}]
 dyn_badge4 [{shapeType=Box, extentX=8.5988, extentY=1.2846, extentZ=6.2423, center=<4.367 -0.2307 0.046>}]
 dyn_mstr_badge [{shapeType=Capsule, length=0.505, radius=0.505}]
 spine_03 [{shapeType=Capsule, length=5, radius=9.5, center=<1.6093 2.2887 0>}]
 spine_04 [
  {shapeType=Capsule, length=12.948, radius=4.7046, center=<5.2205 6.43 -0.0636>, rotation=<rot -0.4443 3.4088 -0.1187>}
  {shapeType=Capsule, length=15, radius=4.75, center=<1.9406 8.7731 0>}
 ]
}
constraints [
 {
  bone1Name=dyn_mstr_badge
  bone2Name=spine_04
  swing1LimitDegrees=45
  swing1Motion=Free
  swing2LimitDegrees=45
  swing2Motion=Free
  twistLimitDegrees=45
  twistMotion=Free
 }
 {bone1Name=dyn_badge1, bone2Name=dyn_mstr_badge, swing1LimitDegrees=15, swing2LimitDegrees=15, twistLimitDegrees=5}
 {bone1Name=dyn_badge2, bone2Name=dyn_badge1, swing1LimitDegrees=15, swing2LimitDegrees=15, twistLimitDegrees=5}
 {bone1Name=dyn_badge3, bone2Name=dyn_badge2, swing1LimitDegrees=15, swing2LimitDegrees=15, twistLimitDegrees=5}
 {bone1Name=dyn_badge4, bone2Name=dyn_badge3, swing1LimitDegrees=15, swing2LimitDegrees=15, twistLimitDegrees=5}
 {
  bone1Name=spine_04
  bone2Name=spine_03
  swing1LimitDegrees=45
  swing1Motion=Free
  swing2LimitDegrees=45
  swing2Motion=Free
  twistLimitDegrees=45
  twistMotion=Free
 }
]
