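# Physics asset for the F_MED_Barefoot female player body, covering secondary motion of its holster and magazine accessories rather than a full ragdoll.
# Two chains hang from a spine_03 capsule through small master capsules: holster strap to holster, and magazine straps to magazines, all box bodies.
# Master-to-spine joints rotate freely, while joints down each chain have narrow swing and twist limits, tighter on the magazine side.
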PhysicsAsset /BRCosmetics/Characters/Player/Female/Medium/Bodies/F_MED_Barefoot/Meshes/F_MED_Barefoot_Physics
bodies {
 dyn_holster [{shapeType=Box, extentX=13.719, extentY=24.6394, extentZ=8.8077, center=<6.86 2.0134 0.0254>, rotation=<rot -12.4019 12.351 -12.0175>}]
 dyn_holsterStrap [{shapeType=Box, extentX=2.51, extentY=7.51, extentZ=1.51, center=<1.0379 -0.0001 0.5178>}]
 dyn_magazineStraps [{shapeType=Box, extentX=2.01, extentY=5.51, extentZ=1.51, center=<0.7829 0.0002 0>}]
 dyn_magazines [{shapeType=Box, extentX=10.4855, extentY=7.7505, extentZ=3.9142, center=<4.7294 0.0778 -0.6757>}]
 dyn_mstr_holsterStrap [{shapeType=Capsule, length=0.505, radius=0.505}]
 dyn_mstr_magazineStraps [{shapeType=Capsule, length=0.505, radius=0.505}]
 spine_03 [{shapeType=Capsule, length=12, radius=8.0833, center=<-2.8499 3.1915 0.6384>, rotation=<rot 0 0 90>}]
}
constraints [
 {bone1Name=dyn_holsterStrap, bone2Name=dyn_mstr_holsterStrap, swing1LimitDegrees=15, swing2LimitDegrees=25, twistLimitDegrees=10}
 {bone1Name=dyn_holster, bone2Name=dyn_holsterStrap, swing1LimitDegrees=15, swing2LimitDegrees=25, twistLimitDegrees=10}
 {bone1Name=dyn_magazineStraps, bone2Name=dyn_mstr_magazineStraps, swing1LimitDegrees=5, swing2LimitDegrees=25, twistLimitDegrees=10}
 {bone1Name=dyn_magazines, bone2Name=dyn_magazineStraps, swing1LimitDegrees=5, swing2LimitDegrees=20, twistLimitDegrees=10}
 {
  bone1Name=dyn_mstr_holsterStrap
  bone2Name=spine_03
  swing1LimitDegrees=45
  swing1Motion=Free
  swing2LimitDegrees=45
  swing2Motion=Free
  twistLimitDegrees=45
  twistMotion=Free
 }
 {
  bone1Name=dyn_mstr_magazineStraps
  bone2Name=spine_03
  swing1LimitDegrees=45
  swing1Motion=Free
  swing2LimitDegrees=45
  swing2Motion=Free
  twistLimitDegrees=45
  twistMotion=Free
 }
]
